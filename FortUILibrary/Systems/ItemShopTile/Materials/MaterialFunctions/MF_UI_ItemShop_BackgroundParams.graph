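# Material function for item shop tile UI that supplies the tile's background color parameters.
# It exposes three vector parameters, Background Color 1 to 3, each with a default color and a sort priority.
# Each parameter's RGB feeds its own output, Color 1 to 3, with no other processing, so alpha is not passed on.
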MaterialFunction /FortUILibrary/Systems/ItemShopTile/Materials/MaterialFunctions/MF_UI_ItemShop_BackgroundParams
graph {nodes [
  FunctionOutput(None=VectorParameter() {parameterName=Background Color 1, sortPriority=10, defaultValue=<rgba 1 0.1094 0.1094 1>}.RGB) {outputName=Color 1}
  FunctionOutput(None=VectorParameter() {parameterName=Background Color 2, sortPriority=20, defaultValue=<rgba 1 0.7236 0.376 1>}.RGB) {outputName=Color 2, sortPriority=1}
  FunctionOutput(None=VectorParameter() {parameterName=Background Color 3, sortPriority=30, defaultValue=<rgba 0.2898 0.9502 1 1>}.RGB) {outputName=Color 3, sortPriority=2}
 ]}
