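# Material function that exposes the shared scalar parameters for item shop offer tile images.
# It wraps anchor point X/Y, zoom level, rounded-corner flag, scale-to-fit-X and image fade alpha, grouped as tile-driven params, plus hover and hover zoom modifier.
# Each parameter feeds straight to its own named output with no math, giving materials one place to pull these values from.
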MaterialFunction /FortUILibrary/Systems/ItemShopTile/Materials/MaterialFunctions/MF_UI_ItemShop_OfferImageParams
graph {nodes [
  FunctionOutput(None=ScalarParameter() {defaultValue=0.5, group=Tile-driven params, parameterName=X Anchor Point, sortPriority=10}) {outputName=X anchor point}
  FunctionOutput(None=ScalarParameter() {defaultValue=0.5, group=Tile-driven params, parameterName=Y Anchor Point, sortPriority=10}) {outputName=Y anchor point, sortPriority=1}
  FunctionOutput(None=ScalarParameter() {defaultValue=1, group=Tile-driven params, parameterName=Zoom Level, sortPriority=20}) {outputName=Zoom level, sortPriority=2}
  FunctionOutput(None=ScalarParameter() {group=Tile-driven params, parameterName=bHasRoundedCorners, sortPriority=20}) {outputName=HasRoundedCorners, sortPriority=4}
  FunctionOutput(None=ScalarParameter() {group=Tile-driven params, parameterName=ScaleToFitX, sortPriority=20}) {outputName=ScaleToFitX, sortPriority=3}
  FunctionOutput(None=ScalarParameter() {group=Tile-driven params, parameterName=FadeImageAlpha, sortPriority=20}) {outputName=Fade image alpha, sortPriority=5}
  FunctionOutput(None=ScalarParameter() {parameterName=Hover}) {outputName=Hover, sortPriority=2}
  FunctionOutput(None=ScalarParameter() {defaultValue=0.05, parameterName=HoverZoomModifier}) {outputName=Hover zoom modifier, sortPriority=2}
 ]}
