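# Material function that computes UVs for Item Shop tile offer images from framing rules in MF_UI_ItemShop_OfferImageParams.
# It builds a UV scale from zoom, tile aspect ratio and Texture Size, with a blend between fitting X and Y.
# An anchor offset compensated for zoom shifts the UVs before they are scaled about center.
# It outputs the final UVs, UV scale and anchored UV position.
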MaterialFunction /FortUILibrary/Systems/ItemShopTile/Materials/MaterialFunctions/MF_UI_ItemShop_OfferImage_RulesetUVs
graph {nodes [
  n1 = NamedRerouteUsage()
  n2 = NamedRerouteUsage()
  n3 = Add(A=Multiply(A=AppendVector(A=@n1, B=@n2)) {constB=-2})
  n4 = NamedRerouteUsage()
  n5 = NamedRerouteUsage()
  n6 = Divide(A=AppendVector(A=@n4, B=@n4), B=Max(A=Multiply(B=@n5) {constA=2}) {constB=0.0001}) {constB=3, desc="anchor point needs to\r\ntake zoom into account"}
  n7 = NamedRerouteDeclaration(None=Add(A=Multiply(A=@n3, B=@n6), B=NamedRerouteUsage())) {name=anchored UVs, nodeColor=<rgba 0 1 0.0156 1>}
  n8 = NamedRerouteUsage()
  n9 = AppendVector(A=@n8, B=@n8)
  n10 = MaterialFunctionCall!(Dimensions) {materialFunction=/FortUILibrary/MaterialFunctions/Basics/MF_UI_GetAspectRatio}
  n11 = LinearInterpolate(A=@n10.H Ratio, B=@n10.H Ratio, Alpha=NamedRerouteUsage())
  n12 = MaterialFunctionCall(Float2=FunctionInput() {inputName=Texture Size, inputType=FunctionInput_Vector2, previewValue=<512 512 0 1>}) {materialFunction=&BreakOutFloat2Components}
  n13 = ComponentMask(None=@n12.R) {r=true}
  n14 = ComponentMask(None=@n12.G) {g=true}
  n15 = Constant() {r=1}
  n16 = AppendVector(A=Divide(A=@n13, B=@n14), B=@n15) {desc=fit Y}
  n17 = AppendVector(A=@n15, B=Divide(A=@n14, B=@n13)) {desc=fit X}
  n18 = NamedRerouteUsage()
  n19 = NamedRerouteDeclaration(None=Multiply(A=Divide(A=@n9, B=@n11), B=LinearInterpolate(A=@n16, B=@n17, Alpha=@n18) {constAlpha=0})) {name=UV scale, nodeColor=<rgba 1 0 0.3047 1>}
  n20 = MaterialFunctionCall(UVs=@n7, Texture Scale=@n19) {materialFunction=&ScaleUVsByCenter}
  FunctionOutput(None=@n20.UVs) {outputName=UVs}
  FunctionOutput(None=@n19) {outputName=UV scale}
  FunctionOutput(None=@n7) {outputName=UV position}
  n21 = MaterialFunctionCall() {materialFunction=&/MF_UI_ItemShop_OfferImageParams}
  NamedRerouteDeclaration(None=@n21.X anchor point) {name=x anchor point, nodeColor=<rgba 0.0312 0 1 1>}
  n22 = NamedRerouteUsage()
  n23 = NamedRerouteUsage()
  NamedRerouteDeclaration(None=LinearInterpolate(A=@n22, B=Add(A=@n22, B=@n23), Alpha=NamedRerouteUsage())) {name=final zoom level, nodeColor=<rgba 1 0 0.8906 1>}
  n24 = MaterialFunctionCall!(Vector2, MaximumNumberOfDigits, UVs, DebugTextLocation RG_UpperRight BA_LowerLeft, Component Spacing) {materialFunction=/Engine/Functions/Engine_MaterialFunctions02/Utility/DebugFloat2Values}
  NamedRerouteDeclaration(None=@n24.ColorCodedOutput) {nodeColor=<rgba 1 0.3984 0 1>}
  NamedRerouteDeclaration(None=@n24.GreyScaleOutput) {name=Name_1, nodeColor=<rgba 0.0781 0 1 1>}
  NamedRerouteDeclaration(None=@n20.0-1 mask) {name=Name_2, nodeColor=<rgba 0.2422 1 0 1>}
  NamedRerouteDeclaration(None=@n21.Y anchor point) {name=y anchor point, nodeColor=<rgba 0.9453 1 0 1>}
  NamedRerouteDeclaration(None=FunctionInput(Preview=TextureCoordinate()) {bUsePreviewValueAsDefault=true, inputName=UVs, inputType=FunctionInput_Vector2}) {name=input UVs, nodeColor=<rgba 1 0 0.8906 1>}
  NamedRerouteDeclaration(None=@n21.ScaleToFitX) {name=scale to fit X, nodeColor=<rgba 0 0.5312 1 1>}
  NamedRerouteDeclaration(None=@n21.Hover) {name=hover, nodeColor=<rgba 0 0.25 1 1>}
  NamedRerouteDeclaration(None=@n21.HasRoundedCorners) {name=has rounded corners, nodeColor=<rgba 0.8516 1 0 1>}
 ]}
