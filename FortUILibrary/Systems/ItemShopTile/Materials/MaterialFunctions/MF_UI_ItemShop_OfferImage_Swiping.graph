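# Material function that builds an animated swipe alpha mask, named for item shop offer images.
# It rotates and flips UV channel 4, then slides it by an offset driven by the Swipe parameter.
# The red channel is steepened into an edge, scaled by a resolution sharpness correction function and Swipe_Sharpness_V, then saturated.
# InvertSwipe blends toward the inverted mask, and FinalAlpha scales the saturated output.
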MaterialFunction /FortUILibrary/Systems/ItemShopTile/Materials/MaterialFunctions/MF_UI_ItemShop_OfferImage_Swiping
graph {nodes [
  n1 = TextureCoordinate() {coordinateIndex=4}
  n2 = Constant() {r=-0.25}
  n3 = Add(
   A=OneMinus(None=Rotator(Coordinate=@n1, Time=@n2) {speed=6.2832})
   B=LinearInterpolate(A=Constant() {r=0.05}, B=Constant() {r=-1.05}, Alpha=ScalarParameter() {group=Animation, parameterName=Swipe})
  )
  n4 = MaterialFunctionCall!(Native_U_PixelSize=Constant() {r=384}) {materialFunction=/FortUILibrary/Foundation/Art_ToolKit/Tools/M_UI_Resolution_Sharpness_Correction}
  n5 = ScalarParameter() {defaultValue=64, group=Animation, parameterName=Swipe_Sharpness_V}
  n6 = Saturate(None=Multiply(A=ComponentMask(None=@n3) {r=true}, B=Multiply(A=@n4.Result, B=@n5)) {constB=32})
  n7 = LinearInterpolate(A=@n6, B=OneMinus(None=@n6), Alpha=ScalarParameter() {group=Global, parameterName=InvertSwipe, sliderMax=1})
  n8 = ScalarParameter() {defaultValue=1, group=Global, parameterName=FinalAlpha, sliderMax=1}
  FunctionOutput(None=Saturate(None=Multiply(A=@n7, B=@n8))) {outputName=Swipe alpha mask}
 ]}
