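# Material function that applies rounded-corner masking to the alpha of Item Shop tile UI.
# CornerRadius from MPC_ItemShop fills all four corners by default, with a scalar lerp switching to an optional per-corner Vector4 override.
# The chosen radii feed MF_UI_CoreUI_RoundedMask in pixel units.
# The final alpha is the minimum of the input alpha and that shape mask.
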MaterialFunction /FortUILibrary/Systems/ItemShopTile/Materials/MaterialFunctions/MF_UI_ItemShop_RoundedCorners
graph {nodes [
  n1 = CollectionParameter() {collection=/FortUILibrary/Systems/ItemShopTile/MPC_ItemShop, parameterName=CornerRadius}
  n2 = LinearInterpolate(
   A=MaterialFunctionCall(R=@n1, G=@n1, B=@n1, A=@n1) {materialFunction=&AppendMany}.RGBA
   B=FunctionInput(Preview=Constant4Vector().RGBA) {bUsePreviewValueAsDefault=true, inputName=Corner Radius Override, inputType=FunctionInput_Vector4, sortPriority=2}
   Alpha=FunctionInput(Preview=Constant()) {bUsePreviewValueAsDefault=true, inputName=bOverrideCornerRadius, inputType=FunctionInput_Scalar, sortPriority=1}
  )
  n3 = Constant() {r=1}
  FunctionOutput(None=Min(
    A=FunctionInput() {inputName=Input Alpha, inputType=FunctionInput_Scalar}
    B=MaterialFunctionCall(Corner Radius=@n2, Radius in Pixels=@n3) {materialFunction=&MF_UI_CoreUI_RoundedMask}.Shape Mask
   )) {outputName=Final Alpha}
 ]}
